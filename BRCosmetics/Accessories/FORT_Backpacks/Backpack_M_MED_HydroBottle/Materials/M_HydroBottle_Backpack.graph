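# Backpack material for the HydroBottle cosmetic that adds animated sci-fi text screens to emissive, wrapped between the shared character parent PreFX and PostFX functions.
# A Tech Art Mask channel gates the screen effect, and step thresholds split three screen areas, each layering text over scanline color.
# Custom UV remapping and floored, sine-modulated time scroll text line by line.
# Bounding-box masks with quantized moving edges animate horizontal text reveal.
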Material /BRCosmetics/Accessories/FORT_Backpacks/Backpack_M_MED_HydroBottle/Materials/M_HydroBottle_Backpack
graph {
 properties {
  bHasPixelAnimation=true
  bUseMaterialAttributes=true
  bUsedWithMeshDeformer=true
  bUsedWithSkeletalMesh=true
  pixelDepthOffsetMode=PDOM_Legacy
  shadingModel=MSM_FromMaterialExpression
  usedShadingModels=Default Lit | Subsurface | Clear Coat | Subsurface Profile | Cloth
 }
 nodes [
  Add()
  n1 = VectorParameter() {
   group=Scifi_Screen
   parameterName=TextTexture_Center01
   channelNames {a=ScaleY, b=ScaleX, g=CenterY, r=CenterX}
   defaultValue=<rgba 0.5 0.5 1 1>
  }
  MaterialExpressionAppendVector_19 = AppendVector(A=@n1.CenterX, B=@n1.CenterX)
  MaterialExpressionAppendVector_20 = AppendVector(A=@n1.ScaleX, B=@n1.ScaleX)
  MaterialExpressionStaticSwitchParameter_0 = StaticSwitchParameter(True=TextureCoordinate() {coordinateIndex=1}, False=TextureCoordinate()) {defaultValue=true, group=Scifi_Screen, parameterName=UseTexcoord01}
  n2 = Custom!(
   UVPoint=@MaterialExpressionAppendVector_19
   UVScale=@MaterialExpressionAppendVector_20
   OriginalUV=@MaterialExpressionStaticSwitchParameter_0
  ) {
   code="float2 DistUV =OriginalUV-UVPoint;\r\nDistUV = DistUV/UVScale + float2(0.5,0.5);\r\n\r\nreturn DistUV;"
   description=CopyToPointUV
   outputType=CMOT_Float2
   inputs [
    {
     inputName=UVPoint
     input {expression=&self:MaterialExpressionAppendVector_19}
    }
    {
     inputName=UVScale
     input {expression=&self:MaterialExpressionAppendVector_20}
    }
    {
     inputName=OriginalUV
     input {expression=&self:MaterialExpressionStaticSwitchParameter_0}
    }
   ]
  }
  n3 = AppendVector(A=Constant(), B=Divide(B=ScalarParameter() {defaultValue=10, group=Scifi_Screen, parameterName=LinesNumber}))
  n4 = ScalarParameter() {defaultValue=-1, group=Scifi_Screen, parameterName=SpeedDirection01}
  n5 = Multiply(A=Time(), B=ScalarParameter() {defaultValue=1, group=Scifi_Screen, parameterName=TextSpeed01})
  n6 = Divide(A=@n5, B=MaterialFunctionCall() {materialFunction=&Pi}.Result)
  n7 = Add(A=Multiply(A=Sine(None=@n6)) {constB=0.5}) {constB=0.5}
  n8 = Add(A=Add(A=@n5, B=Multiply(B=@n7) {constA=1}), B=ScalarParameter() {parameterName=TimeOffset01})
  n9 = Add(
   A=NamedRerouteDeclaration(None=@n2) {name=AdjustUV, nodeColor=<rgba 0 1 0.5313 1>}
   B=Multiply(A=Multiply(A=@n3, B=@n4) {constB=-1}, B=Floor(None=@n8))
  )
  Frac(None=@n9)
  n10 = MaterialFunctionCall() {materialFunction=&MF_CharacterParent_2023_PreFX}
  n11 = GetMaterialAttributes(None=@n10.Result) {attributeGetTypes [#1]}
  n12 = NamedRerouteUsage()
  n13 = SmoothStep(
   Max=ScalarParameter() {defaultValue=0.2, group=Scifi_Screen, parameterName=MaxAmount_TechArtMask}
   Value=NamedRerouteDeclaration(None=ChannelMaskParameter(None=@n10.Tech Art Mask) {group=Scifi_Screen, maskChannel=Green, parameterName=ScreenChannel, defaultValue=<rgba 0 1 0 0>}.R) {name=TechArtMask, nodeColor=<rgba 0 0.8359 1 1>}
  ) {constMax=0.2}
  MaterialFunctionCall(
   Material=SetMaterialAttributes(MaterialAttributes=@n11.MaterialAttributes, Emissive Color=LinearInterpolate(A=@n11.Emissive Color, B=@n12, Alpha=@n13)) {attributeSetTypes [#1]}
   M=@n10.M
   Pre Skinned Local Position=@n10.M
   Distance Blend=@n10.M
   Fresnel=@n10.M
  ) {materialFunction=&MF_CharacterParent_2023_PostFX}
  NamedRerouteDeclaration(None=@n1.ScaleY) {name=ScaleY, nodeColor=<rgba 0.2422 0 1 1>}
  n14 = NamedRerouteUsage()
  n15 = Constant() {r=1}
  NamedRerouteDeclaration(None=If(
    A=ComponentMask(None=@n14) {g=true}
    B=ScalarParameter() {group=Scifi_Screen, parameterName=BottomLine}
    "A > B"=Constant()
    "A == B"=@n15
    "A < B"=@n15
   )) {name=BottomMask, nodeColor=<rgba 0.4531 1 0 1>}
  n16 = NamedRerouteUsage()
  n17 = Step(X=@n16) {constY=0.9}
  NamedRerouteDeclaration(None=@n17) {name=Area01, nodeColor=<rgba 0.5 1 0 1>}
  n18 = VectorParameter() {group=Scifi_Screen, parameterName=BackgroundDarkGreen, defaultValue=<rgba 0.8388 0.3596 0 1>}
  n19 = VectorParameter() {group=Scifi_Screen, parameterName=BackGroundGreen, defaultValue=<rgba 0.8388 0.5354 0.0994 1>}
  n20 = TextureSampleParameter2D(UVs=Multiply(A=NamedRerouteUsage(), B=ScalarParameter() {defaultValue=1, group=Scifi_Screen, parameterName=ScanLineTextureTilling01})) {
   group=Scifi_Screen
   parameterName=ScanLine
   samplerSource=SSM_Wrap_WorldGroupSettings
   texture=/BRCosmetics/Weapons/FORT_Melee/Pickaxe_HydroBottle/Textures/T_ScanLines
  }
  n21 = TextureSampleParameter2D(UVs=@n9) {
   group=Scifi_Screen
   parameterName=Text_Blur01
   samplerSource=SSM_Wrap_WorldGroupSettings
   samplerType=SAMPLERTYPE_Masks
   texture=/BRCosmetics/Weapons/FORT_Melee/Pickaxe_HydroBottle/Textures/T_ScreenTilable
  }
  n22 = LinearInterpolate(A=@n21.R, B=Multiply(A=@n21.R, B=NamedRerouteUsage()), Alpha=NamedRerouteUsage())
  n23 = Add(
   A=NamedRerouteDeclaration(None=LinearInterpolate(A=@n18.RGB, B=@n19.RGB, Alpha=@n20.R)) {name=ScreenBackGround, nodeColor=<rgba 1 0.8438 0 1>}
   B=Multiply(A=VectorParameter() {group=Scifi_Screen, parameterName=TextColor, defaultValue=<rgba 0.8388 0.5354 0.0994 1>}.RGB, B=Add(A=@n22, B=@n21.G))
  )
  n24 = Add(A=NamedRerouteUsage(), B=@n23)
  NamedRerouteDeclaration(None=Multiply(A=StaticSwitchParameter(True=@n24, False=@n23) {group=Scifi_Screen, parameterName=DebugArea}, B=NamedRerouteUsage())) {name=ScreenColor01, nodeColor=<rgba 1 0.4453 0 1>}
  NamedRerouteDeclaration(None=Multiply(A=OneMinus(None=@n17), B=Step(X=@n16) {constY=0.4})) {name=Area02, nodeColor=<rgba 0.125 0 1 1>}
  n25 = NamedRerouteUsage()
  n26 = ComponentMask(None=@n25) {r=true}
  n27 = VectorParameter() {group=Scifi_Screen, parameterName=TextBoundingBox02, channelNames {a=YMax, b=YMin, g=XMax, r=XMin}, defaultValue=<rgba 0 1 0 1>}
  n28 = Constant()
  n29 = If(
   A=@n26
   B=NamedRerouteDeclaration(None=@n27.XMin) {name=XMin_1, nodeColor=<rgba 0.2891 0 1 1>}
   "A > B"=Constant() {r=1}
   "A == B"=@n28
   "A < B"=@n28
  )
  n30 = Constant()
  n31 = OneMinus(None=If(A=@n26, B=NamedRerouteUsage(), "A > B"=@n30, "A == B"=@n30, "A < B"=Constant() {r=1}))
  n32 = ComponentMask(None=@n25) {g=true}
  n33 = Constant()
  n34 = Constant()
  n35 = Multiply(
   A=If(A=@n32, B=@n27.YMin, "A > B"=Constant() {r=1}, "A == B"=@n33, "A < B"=@n33)
   B=If(A=@n32, B=@n27.YMax, "A > B"=@n34, "A == B"=@n34, "A < B"=Constant() {r=1})
  )
  NamedRerouteDeclaration(None=Multiply(A=Multiply(A=@n29, B=@n31), B=@n35)) {name=AreaMaskDynamic_1, nodeColor=<rgba 1 0.2109 0 1>}
  NamedRerouteDeclaration(None=@n27.XMax) {name=XMax_1, nodeColor=<rgba 0.2891 0 1 1>}
  n36 = NamedRerouteUsage()
  n37 = NamedRerouteUsage()
  NamedRerouteDeclaration(None=Add(A=Add(A=@n36, B=@n37), B=NamedRerouteUsage())) {name=NewEM, nodeColor=<rgba 0.6172 1 0 1>}
  n38 = Multiply(A=Time(), B=ScalarParameter() {defaultValue=1, group=Scifi_Screen, parameterName=TextSpeed02})
  n39 = Divide(A=@n38, B=MaterialFunctionCall() {materialFunction=&Pi}.Result)
  n40 = Add(A=Multiply(A=Sine(None=@n39)) {constB=0.5}) {constB=0.5}
  n41 = Add(A=Add(A=@n38, B=Multiply(B=@n40) {constA=1}), B=ScalarParameter() {parameterName=TimeOffset02})
  NamedRerouteDeclaration(None=@n41) {name=TimeAdj_1, nodeColor=<rgba 1 0 0.3516 1>}
  n42 = ScalarParameter() {defaultValue=20, parameterName=HorizontalDivisor}
  n43 = Multiply(A=NamedRerouteUsage(), B=@n42)
  n44 = Divide(A=Fmod(A=Floor(None=@n43), B=@n42), B=@n42)
  NamedRerouteDeclaration(None=LinearInterpolate(A=NamedRerouteUsage(), B=NamedRerouteUsage(), Alpha=OneMinus(None=@n44))) {name=XMax_Line_1, nodeColor=<rgba 0 0.6719 1 1>}
  n45 = Constant()
  n46 = If(A=@n26, B=NamedRerouteUsage(), "A > B"=@n45, "A == B"=@n45, "A < B"=Constant() {r=1})
  NamedRerouteDeclaration(None=Multiply(A=Multiply(A=@n29, B=@n46), B=@n35)) {name=AreaMaskStatic_1, nodeColor=<rgba 0.7578 1 0 1>}
  n47 = VectorParameter() {
   group=Scifi_Screen
   parameterName=TextTexture_Center02
   channelNames {a=ScaleY, b=ScaleX, g=CenterY, r=CenterX}
   defaultValue=<rgba 0.5 0.5 1 1>
  }
  NamedRerouteDeclaration(None=@n47.ScaleY) {name=ScaleY_1, nodeColor=<rgba 0.2422 0 1 1>}
  n48 = NamedRerouteUsage()
  n49 = Constant() {r=1}
  NamedRerouteDeclaration(None=If(
    A=ComponentMask(None=@n48) {g=true}
    B=ScalarParameter() {group=Scifi_Screen, parameterName=BottomLine}
    "A > B"=Constant()
    "A == B"=@n49
    "A < B"=@n49
   )) {name=BottomMask_1, nodeColor=<rgba 0.4531 1 0 1>}
  n50 = VectorParameter() {group=Scifi_Screen, parameterName=BackgroundDarkOrange, defaultValue=<rgba 0.8388 0.3596 0 1>}
  n51 = VectorParameter() {group=Scifi_Screen, parameterName=BackGroundYellow, defaultValue=<rgba 0.8388 0.5354 0.0994 1>}
  n52 = TextureSampleParameter2D(UVs=Multiply(A=NamedRerouteUsage(), B=ScalarParameter() {defaultValue=0.5, group=Scifi_Screen, parameterName=ScanLineTextureTilling02})) {
   group=Scifi_Screen
   parameterName=ScanLine
   samplerSource=SSM_Wrap_WorldGroupSettings
   texture=&T_ScanLines
  }
  n53 = VectorParameter() {group=Scifi_Screen, parameterName=TextColor_Yellow, defaultValue=<rgba 0.8388 0.5354 0.0994 1>}
  MaterialExpressionAppendVector_1 = AppendVector(A=@n47.CenterX, B=@n47.CenterX)
  MaterialExpressionAppendVector_2 = AppendVector(A=@n47.ScaleX, B=@n47.ScaleX)
  MaterialExpressionNamedRerouteUsage_31 = NamedRerouteUsage()
  n54 = Custom(UVPoint=@MaterialExpressionAppendVector_1, UVScale=@MaterialExpressionAppendVector_2, OriginalUV=@MaterialExpressionNamedRerouteUsage_31) {
   code="float2 DistUV =OriginalUV-UVPoint;\r\nDistUV = DistUV/UVScale + float2(0.5,0.5);\r\n\r\nreturn DistUV;"
   description=CopyToPointUV
   outputType=CMOT_Float2
   inputs [
    {
     inputName=UVPoint
     input {expression=&self:MaterialExpressionAppendVector_1}
    }
    {
     inputName=UVScale
     input {expression=&self:MaterialExpressionAppendVector_2}
    }
    {
     inputName=OriginalUV
     input {expression=&self:MaterialExpressionNamedRerouteUsage_31}
    }
   ]
  }
  n55 = ScalarParameter() {defaultValue=10, group=Scifi_Screen, parameterName=LinesNumber}
  n56 = Multiply(A=AppendVector(A=Constant(), B=Divide(B=@n55)), B=ScalarParameter() {defaultValue=-1, group=Scifi_Screen, parameterName=SpeedDirection02}) {constB=-1}
  n57 = Floor(None=@n41)
  n58 = TextureSampleParameter2D(UVs=Add(A=NamedRerouteDeclaration(None=@n54) {name=AdjustUV_1, nodeColor=<rgba 0 1 0.5312 1>}, B=Multiply(A=@n56, B=@n57))) {
   group=Scifi_Screen
   parameterName=Text_Blur02
   samplerSource=SSM_Wrap_WorldGroupSettings
   samplerType=SAMPLERTYPE_Masks
   texture=&T_ScreenTilable
  }
  n59 = LinearInterpolate(A=@n58.R, B=Multiply(A=@n58.R, B=NamedRerouteUsage()), Alpha=NamedRerouteUsage())
  n60 = Add(
   A=NamedRerouteDeclaration(None=LinearInterpolate(A=@n50.RGB, B=@n51.RGB, Alpha=@n52.R)) {name=ScreenBackGround_Yellow, nodeColor=<rgba 1 0.8438 0 1>}
   B=Multiply(A=@n53.RGB, B=Add(A=@n59, B=@n58.G))
  )
  n61 = Add(A=NamedRerouteUsage(), B=@n60)
  NamedRerouteDeclaration(None=Multiply(A=StaticSwitchParameter(True=@n61, False=@n60) {group=Scifi_Screen, parameterName=DebugArea}, B=NamedRerouteUsage())) {name=ScreenColor02, nodeColor=<rgba 1 0.4453 0 1>}
  NamedRerouteDeclaration(None=@MaterialExpressionStaticSwitchParameter_0) {name=BaseCoord, nodeColor=<rgba 1 0.1641 0 1>}
  NamedRerouteDeclaration(None=Step(Y=@n16) {constX=0.4, constY=0.4}) {name=Area03, nodeColor=<rgba 0 0.5781 1 1>}
  n62 = NamedRerouteUsage()
  n63 = ComponentMask(None=@n62) {r=true}
  n64 = VectorParameter() {group=Scifi_Screen, parameterName=TextBoundingBox03, channelNames {a=YMax, b=YMin, g=XMax, r=XMin}, defaultValue=<rgba 0 1 0 1>}
  n65 = Constant()
  n66 = If(
   A=@n63
   B=NamedRerouteDeclaration(None=@n64.XMin) {name=XMin_2, nodeColor=<rgba 0.2891 0 1 1>}
   "A > B"=Constant() {r=1}
   "A == B"=@n65
   "A < B"=@n65
  )
  n67 = Constant()
  n68 = OneMinus(None=If(A=@n63, B=NamedRerouteUsage(), "A > B"=@n67, "A == B"=@n67, "A < B"=Constant() {r=1}))
  n69 = ComponentMask(None=@n62) {g=true}
  n70 = Constant()
  n71 = Constant()
  n72 = Multiply(
   A=If(A=@n69, B=@n64.YMin, "A > B"=Constant() {r=1}, "A == B"=@n70, "A < B"=@n70)
   B=If(A=@n69, B=@n64.YMax, "A > B"=@n71, "A == B"=@n71, "A < B"=Constant() {r=1})
  )
  NamedRerouteDeclaration(None=Multiply(A=Multiply(A=@n66, B=@n68), B=@n72)) {name=AreaMaskDynamic_2, nodeColor=<rgba 1 0.2109 0 1>}
  NamedRerouteDeclaration(None=@n64.XMax) {name=XMax_2, nodeColor=<rgba 0.2891 0 1 1>}
  n73 = Multiply(A=Time(), B=ScalarParameter() {defaultValue=1, group=Scifi_Screen, parameterName=TextSpeed03})
  n74 = Divide(A=@n73, B=MaterialFunctionCall() {materialFunction=&Pi}.Result)
  n75 = Add(A=Multiply(A=Sine(None=@n74)) {constB=0.5}) {constB=0.5}
  n76 = Add(A=Add(A=@n73, B=Multiply(B=@n75) {constA=1}), B=ScalarParameter() {parameterName=TimeOffset03})
  NamedRerouteDeclaration(None=@n76) {name=TimeAdj_2, nodeColor=<rgba 1 0 0.3516 1>}
  n77 = ScalarParameter() {defaultValue=20, parameterName=HorizontalDivisor}
  n78 = Multiply(A=NamedRerouteUsage(), B=@n77)
  n79 = Divide(A=Fmod(A=Floor(None=@n78), B=@n77), B=@n77)
  NamedRerouteDeclaration(None=LinearInterpolate(A=NamedRerouteUsage(), B=NamedRerouteUsage(), Alpha=OneMinus(None=@n79))) {name=XMax_Line_2, nodeColor=<rgba 0 0.6719 1 1>}
  n80 = Constant()
  n81 = If(A=@n63, B=NamedRerouteUsage(), "A > B"=@n80, "A == B"=@n80, "A < B"=Constant() {r=1})
  NamedRerouteDeclaration(None=Multiply(A=Multiply(A=@n66, B=@n81), B=@n72)) {name=AreaMaskStatic_2, nodeColor=<rgba 0.7578 1 0 1>}
  n82 = VectorParameter() {
   group=Scifi_Screen
   parameterName=TextTexture_Center03
   channelNames {a=ScaleY, b=ScaleX, g=CenterY, r=CenterX}
   defaultValue=<rgba 0.5 0.5 1 1>
  }
  NamedRerouteDeclaration(None=@n82.ScaleY) {name=ScaleY_2, nodeColor=<rgba 0.2422 0 1 1>}
  n83 = NamedRerouteUsage()
  n84 = Constant() {r=1}
  NamedRerouteDeclaration(None=If(
    A=ComponentMask(None=@n83) {g=true}
    B=ScalarParameter() {group=Scifi_Screen, parameterName=BottomLine}
    "A > B"=Constant()
    "A == B"=@n84
    "A < B"=@n84
   )) {name=BottomMask_2, nodeColor=<rgba 0.4531 1 0 1>}
  MaterialExpressionAppendVector_4 = AppendVector(A=@n82.CenterX, B=@n82.CenterX)
  MaterialExpressionAppendVector_5 = AppendVector(A=@n82.ScaleX, B=@n82.ScaleX)
  MaterialExpressionNamedRerouteUsage_39 = NamedRerouteUsage()
  n85 = Custom(UVPoint=@MaterialExpressionAppendVector_4, UVScale=@MaterialExpressionAppendVector_5, OriginalUV=@MaterialExpressionNamedRerouteUsage_39) {
   code="float2 DistUV =OriginalUV-UVPoint;\r\nDistUV = DistUV/UVScale + float2(0.5,0.5);\r\n\r\nreturn DistUV;"
   description=CopyToPointUV
   outputType=CMOT_Float2
   inputs [
    {
     inputName=UVPoint
     input {expression=&self:MaterialExpressionAppendVector_4}
    }
    {
     inputName=UVScale
     input {expression=&self:MaterialExpressionAppendVector_5}
    }
    {
     inputName=OriginalUV
     input {expression=&self:MaterialExpressionNamedRerouteUsage_39}
    }
   ]
  }
  n86 = ScalarParameter() {defaultValue=10, group=Scifi_Screen, parameterName=LinesNumber}
  n87 = Multiply(A=AppendVector(A=Constant(), B=Divide(B=@n86)), B=ScalarParameter() {defaultValue=-1, group=Scifi_Screen, parameterName=SpeedDirection03}) {constB=-1}
  n88 = Floor(None=@n76)
  n89 = TextureSampleParameter2D(UVs=Add(A=NamedRerouteDeclaration(None=@n85) {name=AdjustUV_2, nodeColor=<rgba 0 1 0.5312 1>}, B=Multiply(A=@n87, B=@n88))) {
   group=Scifi_Screen
   parameterName=Text_Blur02
   samplerSource=SSM_Wrap_WorldGroupSettings
   samplerType=SAMPLERTYPE_Masks
   texture=&T_ScreenTilable
  }
  n90 = LinearInterpolate(A=@n89.R, B=Multiply(A=@n89.R, B=NamedRerouteUsage()), Alpha=NamedRerouteUsage())
  n91 = Add(A=NamedRerouteUsage(), B=Multiply(A=NamedRerouteUsage(), B=Add(A=@n90, B=@n89.G)))
  n92 = Add(A=NamedRerouteUsage(), B=@n91)
  NamedRerouteDeclaration(None=Multiply(A=StaticSwitchParameter(True=@n92, False=@n91) {group=Scifi_Screen, parameterName=DebugArea}, B=NamedRerouteUsage())) {name=ScreenColor03, nodeColor=<rgba 1 0.4453 0 1>}
  NamedRerouteDeclaration(None=@n53.RGB) {name=Text_Yellow, nodeColor=<rgba 0.3594 1 0 1>}
  n93 = NamedRerouteUsage()
  n94 = ComponentMask(None=@n93) {r=true}
  n95 = VectorParameter() {group=Scifi_Screen, parameterName=TextBoundingBox01, channelNames {a=YMax, b=YMin, g=XMax, r=XMin}, defaultValue=<rgba 0 1 0 1>}
  n96 = Constant()
  n97 = If(
   A=@n94
   B=NamedRerouteDeclaration(None=@n95.XMin) {name=XMin, nodeColor=<rgba 0.2891 0 1 1>}
   "A > B"=Constant() {r=1}
   "A == B"=@n96
   "A < B"=@n96
  )
  n98 = Constant()
  n99 = If(A=@n94, B=NamedRerouteUsage(), "A > B"=@n98, "A == B"=@n98, "A < B"=Constant() {r=1})
  n100 = ComponentMask(None=@n93) {g=true}
  n101 = Constant()
  n102 = Constant()
  n103 = Multiply(
   A=If(A=@n100, B=@n95.YMin, "A > B"=Constant() {r=1}, "A == B"=@n101, "A < B"=@n101)
   B=If(A=@n100, B=@n95.YMax, "A > B"=@n102, "A == B"=@n102, "A < B"=Constant() {r=1})
  )
  NamedRerouteDeclaration(None=Multiply(A=Multiply(A=@n97, B=@n99), B=@n103)) {name=AreaMaskDynamic, nodeColor=<rgba 1 0.2109 0 1>}
  NamedRerouteDeclaration(None=@n95.XMax) {name=XMax, nodeColor=<rgba 0.2891 0 1 1>}
  NamedRerouteDeclaration(None=@n8) {name=TimeAdj, nodeColor=<rgba 1 0 0.3516 1>}
  n104 = ScalarParameter() {defaultValue=20, parameterName=HorizontalDivisor}
  n105 = Multiply(A=NamedRerouteUsage(), B=@n104)
  n106 = Divide(A=Fmod(A=Floor(None=@n105), B=@n104), B=@n104)
  NamedRerouteDeclaration(None=LinearInterpolate(A=NamedRerouteUsage(), B=NamedRerouteUsage(), Alpha=@n106)) {name=XMax_Line, nodeColor=<rgba 0 0.6719 1 1>}
  n107 = Constant()
  n108 = If(A=@n94, B=NamedRerouteUsage(), "A > B"=@n107, "A == B"=@n107, "A < B"=Constant() {r=1})
  NamedRerouteDeclaration(None=Multiply(A=Multiply(A=@n97, B=@n108), B=@n103)) {name=AreaMaskStatic, nodeColor=<rgba 0.7578 1 0 1>}
  OneMinus(None=@n106)
 ]
}
