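# Physics asset for the HydroBottle medium male backpack cosmetic, defining collision and joint limits for its dangling wire bones.
# Small capsules form wire chains with limited swing and mostly tight twist, most starting at their own base body.
# The multi-capsule backpack base links to wire 4 and parts of wire 7.
# Boxes cover the pack and the wire 4 base, and neither has a constraint.
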PhysicsAsset /BRCosmetics/Accessories/FORT_Backpacks/Backpack_M_MED_HydroBottle/Meshes/M_MED_HydroBottle_Pack_Physics_off
bodies {
 dyn_backpack_base [
  {shapeType=Capsule, length=1, radius=9.5, shapeName=dyn_backpack_base_capsule, center=<0 -10.3622 -3.4499>}
  {
   shapeType=Capsule
   length=8.5
   radius=2.5
   shapeName=dyn_backpack_base_capsule
   center=<-4.4047 -18.8405 -0.26>
   rotation=<rot 1.7279 -9.8511 9.851>
  }
  {shapeType=Capsule, length=8, radius=2.5, shapeName=dyn_backpack_base_capsule, center=<6.9723 -16.5581 0.8802>, rotation=<rot 0 0 10>}
 ]
 dyn_wires_1_1 [{shapeType=Capsule, length=3, radius=1, shapeName=dyn_wires_1_1_capsule, center=<1.9252 0.8124 0>, rotation=<rot -90 104.1418 -114.1425>}]
 dyn_wires_1_2 [{shapeType=Capsule, length=2.5, radius=1.5, shapeName=dyn_wires_1_2_capsule, center=<1.9247 0.3134 0>, rotation=<rot -90 179.5524 -174.5531>}]
 dyn_wires_1_3 [{shapeType=Capsule, length=4, radius=1.5, shapeName=dyn_wires_1_3_capsule, center=<3.3267 -0.7333 0>, rotation=<rot -90 19.3866 -19.3859>}]
 dyn_wires_1_base [{shapeType=Capsule, length=0, radius=1.5, shapeName=dyn_wires_1_base_capsule}]
 dyn_wires_2_1 [{shapeType=Capsule, length=2.5, radius=1, shapeName=dyn_wires_2_1_capsule, center=<2.2136 -0.6962 0>, rotation=<rot 90 83.5155 68.5145>}]
 dyn_wires_2_2 [{shapeType=Capsule, length=2, radius=1.5, shapeName=dyn_wires_2_2_capsule, center=<1.9145 0.258 0>, rotation=<rot 90 166.7374 166.7368>}]
 dyn_wires_2_3 [{shapeType=Capsule, length=3.5, radius=1.5, shapeName=dyn_wires_2_3_capsule, center=<3.5384 -0.6722 0>, rotation=<rot 90 0 -0.0008>}]
 dyn_wires_2_base [{shapeType=Capsule, length=0, radius=1.5, shapeName=dyn_wires_2_base_capsule}]
 dyn_wires_3_1 [{shapeType=Capsule, length=1.5, radius=1, shapeName=dyn_wires_3_1_capsule, center=<1.5583 0.33 -0.8317>, rotation=<rot 59.9998 0.0008 -10>}]
 dyn_wires_3_2 [{
   shapeType=Capsule
   length=2.5
   radius=1
   shapeName=dyn_wires_3_2_capsule
   center=<2.0085 -0.5128 -0.5642>
   rotation=<rot 74.9994 0.0005 15.0001>
  }]
 dyn_wires_3_base [{shapeType=Capsule, length=0, radius=1.5, shapeName=dyn_wires_3_base_capsule}]
 dyn_wires_4_1 [{shapeType=Capsule, length=4, radius=1.5, shapeName=dyn_wires_4_1_capsule, center=<3.2364 0 0>, rotation=<rot 90 -179.9983 -179.9986>}]
 dyn_wires_4_2 [{shapeType=Capsule, length=4, radius=1.5, shapeName=dyn_wires_4_2_capsule, center=<3.2364 0 0>, rotation=<rot 90 179.9983 179.998>}]
 dyn_wires_4_base [{
   shapeType=Box
   extentX=18.5
   extentY=4.5
   extentZ=10
   shapeName=dyn_wires_4_base_box
   center=<-7.7446 -0.1076 2.8042>
   rotation=<rot 0 5.0001 0>
  }]
 dyn_wires_5_1 [{shapeType=Capsule, length=2.5, radius=1.5, shapeName=dyn_wires_5_1_capsule, center=<2.7729 0 0>, rotation=<rot 90 -179.9852 -179.985>}]
 dyn_wires_5_2 [{shapeType=Capsule, length=2.5, radius=1.5, shapeName=dyn_wires_5_2_capsule, center=<2.4513 0 0>, rotation=<rot 90 166.7373 166.7376>}]
 dyn_wires_5_3 [{shapeType=Capsule, length=5, radius=2.5, shapeName=dyn_wires_5_3_capsule, center=<6.0498 0 -0.9076>, rotation=<rot 80 -0.0003 0.0002>}]
 dyn_wires_5_base [{shapeType=Capsule, length=0, radius=1.5, shapeName=dyn_wires_5_base_capsule}]
 dyn_wires_6_1 [{shapeType=Capsule, length=7.25, radius=0.75, shapeName=dyn_wires_6_1_capsule, center=<3.9054 0 0>, rotation=<rot -90 179.7626 -179.7624>}]
 dyn_wires_6_2 [{shapeType=Capsule, length=5.75, radius=1, shapeName=dyn_wires_6_2_capsule, center=<2.4625 0 -0.236>, rotation=<rot -90 179.9704 -179.97>}]
 dyn_wires_6_3 [{shapeType=Capsule, length=6.25, radius=1.25, shapeName=dyn_wires_6_3_capsule, center=<3.9052 0 0>, rotation=<rot -90 -180 -179.9998>}]
 dyn_wires_6_base [{shapeType=Capsule, length=0, radius=1.5, shapeName=dyn_wires_6_base_capsule}]
 dyn_wires_7_1_1 [{shapeType=Capsule, length=9.75, radius=0.75, shapeName=dyn_wires_7_1_1_capsule, center=<-0.3303 -0.5823 -6.5256>}]
 dyn_wires_7_1_2 [{shapeType=Capsule, length=9.5, radius=1.25, shapeName=dyn_wires_7_1_2_capsule, center=<0.1585 -0.1167 -2.3499>, rotation=<rot 5 0 0>}]
 dyn_wires_7_1_base [{shapeType=Capsule, length=0, radius=1.5, shapeName=dyn_wires_7_1_base_capsule}]
 dyn_wires_7_2 [{shapeType=Capsule, length=7, radius=2.75, shapeName=dyn_wires_7_2_capsule, center=<0 0 0.143>, rotation=<rot 90 -179.9604 -179.9605>}]
 dyn_wires_7_3 [{shapeType=Capsule, length=2, radius=1.5, shapeName=dyn_wires_7_3_capsule, center=<0 0 -4.5978>, rotation=<rot 0.0002 90 -90>}]
 pack [{shapeType=Box, extentX=16.5, extentY=7.5, extentZ=19, shapeName=pack_box, center=<0 -1.323 0>}]
}
constraints [
 {bone1Name=dyn_wires_7_3, bone2Name=dyn_backpack_base, swing1LimitDegrees=10, swing2LimitDegrees=10, twistLimitDegrees=2}
 {bone1Name=dyn_wires_1_1, bone2Name=dyn_wires_1_base, swing1LimitDegrees=10, swing2LimitDegrees=5, twistLimitDegrees=2}
 {bone1Name=dyn_wires_1_2, bone2Name=dyn_wires_1_1, swing1LimitDegrees=30, swing2LimitDegrees=15, twistLimitDegrees=2}
 {bone1Name=dyn_wires_1_3, bone2Name=dyn_wires_1_2, swing1LimitDegrees=20, swing2LimitDegrees=15, twistLimitDegrees=2}
 {bone1Name=dyn_wires_2_1, bone2Name=dyn_wires_2_base, swing1LimitDegrees=20, swing2LimitDegrees=5, twistLimitDegrees=2}
 {bone1Name=dyn_wires_2_2, bone2Name=dyn_wires_2_1, swing1LimitDegrees=30, swing2LimitDegrees=10, twistLimitDegrees=2}
 {bone1Name=dyn_wires_2_3, bone2Name=dyn_wires_2_2, swing1LimitDegrees=10, swing2LimitDegrees=20, twistLimitDegrees=2}
 {bone1Name=dyn_backpack_base, bone2Name=dyn_wires_7_2, swing1LimitDegrees=20, swing2LimitDegrees=10, twistLimitDegrees=20}
 {bone1Name=dyn_wires_3_1, bone2Name=dyn_wires_3_base, swing1LimitDegrees=45, swing2LimitDegrees=25, twistLimitDegrees=2}
 {bone1Name=dyn_wires_3_2, bone2Name=dyn_wires_3_1, swing1LimitDegrees=10, swing2LimitDegrees=10, twistLimitDegrees=2}
 {bone1Name=dyn_wires_4_2, bone2Name=dyn_wires_4_1, swing1LimitDegrees=20, swing2LimitDegrees=20, twistLimitDegrees=5}
 {bone1Name=dyn_wires_5_1, bone2Name=dyn_wires_5_base, swing1LimitDegrees=30, swing2LimitDegrees=15, twistLimitDegrees=5}
 {bone1Name=dyn_wires_5_2, bone2Name=dyn_wires_5_1, swing1LimitDegrees=27.5, swing2LimitDegrees=10, twistLimitDegrees=5}
 {bone1Name=dyn_wires_5_3, bone2Name=dyn_wires_5_2, swing1LimitDegrees=15, swing2LimitDegrees=10, twistLimitDegrees=5}
 {bone1Name=dyn_wires_6_1, bone2Name=dyn_wires_6_base, swing1LimitDegrees=25, swing2LimitDegrees=10, twistLimitDegrees=10}
 {bone1Name=dyn_wires_6_2, bone2Name=dyn_wires_6_1, swing1LimitDegrees=12.5, swing2LimitDegrees=10, twistLimitDegrees=7.5}
 {bone1Name=dyn_wires_6_3, bone2Name=dyn_wires_6_2, swing1LimitDegrees=15, swing2LimitDegrees=10, twistLimitDegrees=5}
 {bone1Name=dyn_wires_7_1_1, bone2Name=dyn_wires_7_1_base, swing1LimitDegrees=25, swing2LimitDegrees=5, twistLimitDegrees=1}
 {bone1Name=dyn_wires_7_1_2, bone2Name=dyn_wires_7_1_1, swing1LimitDegrees=20, swing2LimitDegrees=7.5, twistLimitDegrees=5}
 {bone1Name=dyn_wires_4_1, bone2Name=dyn_backpack_base, swing1LimitDegrees=20, swing2LimitDegrees=20, twistLimitDegrees=5}
]
